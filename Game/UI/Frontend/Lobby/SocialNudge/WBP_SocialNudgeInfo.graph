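# Frontend lobby social nudge widget showing another player's names, platform icons and presence status.
# Functions set both names, color status text from an InGame/Default color map, and hide the secondary row when the nickname is empty.
# The secondary icon always gets the platform; the primary gets it only without a nickname.
# The visibility update always collapses the primary icon and last-played text via a constant-false select; event graph stubs are empty.
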WidgetBlueprint /Game/UI/Frontend/Lobby/SocialNudge/WBP_SocialNudgeInfo
graphs [
 EventGraph {nodes [
   Event PreConstruct()
   Event Construct()
   Event Tick()
  ]}
 UpdatePlatformIcon {nodes [
   n1 = UpdatePlatformIcon!() -> (then, PlatformIconTag_DisplayName:String, PlatformIconTag_Nickname:String, PlatformIconTag_PlatformIconTag:String) {is=title K2Node_FunctionEntry}
   > SetPlatformPrefixIcon!(
    execute
    Image:Image Object Reference=Get Image_PlatformIcon_Secondary() -> (Image_PlatformIcon_Secondary:Image Object Reference)
    OtherPlayerPlatform:String=@n1.PlatformIconTag_PlatformIconTag
   ) -> (then)
   > n2 = Branch(Condition=IsEmpty(InString=@n1.PlatformIconTag_Nickname))
   n3 = Get Image_PlatformIcon_Primary() -> (Image_PlatformIcon_Primary:Image Object Reference)
   > SetPlatformPrefixIcon(Image=@n3, OtherPlayerPlatform=@n1.PlatformIconTag_PlatformIconTag)
   SetPlatformPrefixIcon(execute=@n2.else, Image=@n3)
  ]}
 UpdatePresenceTextColor {nodes [
   n1 = UpdatePresenceTextColor!() -> (then, IsInGame:bool=~) {is=title K2Node_FunctionEntry}
   n2 = Find(
    TargetMap:Map of Strings to Linear Color Structures=Get PresenceTextColors() -> (PresenceTextColors:Map of Strings to Linear Color Structures)
    Key=SelectString(A=InGame, B=Default, bPickA=@n1)
   ) -> (Value:Linear Color Structure)
   > Branch(Condition=@n2.ReturnValue)
   > SetColorAndOpacity!(
    execute
    self=Get Text_InGameStatus() -> (Text_InGameStatus:Common Text Object Reference)
    InColorAndOpacity_SpecifiedColor:Linear Color Structure=@n2
    InColorAndOpacity_ColorUseRule:ESlateColorStylingMode Enum
   ) -> (then)
  ]}
 UpdatePrimaryName {nodes [
   n1 = UpdatePrimaryName!() -> (then, PrimaryName:String) {is=title K2Node_FunctionEntry}
   > Set Text(Text="To Text (String)"(InString=@n1), self=Get Text_Primary() -> (Text_Primary:Common Text Object Reference))
  ]}
 UpdateSecondaryInfoVisibility {nodes [
   n1 = UpdateSecondaryInfoVisibility!() -> (then, Nickname:String) {is=title K2Node_FunctionEntry}
   > SetVisibility(
    self=Get HBox_SecondaryInfo() -> (HBox_SecondaryInfo:Horizontal Box Object Reference)
    InVisibility=Select(Option 0=SelfHitTestInvisible, Option 1=Collapsed, Index=IsEmpty(InString=@n1))
   )
   > SetVisibility(
    self=[Get Image_PlatformIcon_Primary(), Get Text_LastPlayed() -> (Text_LastPlayed:Common Text Object Reference)]
    InVisibility=Select(Option 0=Collapsed, Option 1=SelfHitTestInvisible, Index=false)
   )
  ]}
 UpdateSecondaryName {nodes [
   n1 = UpdateSecondaryName!() -> (then, SecondaryName:String) {is=title K2Node_FunctionEntry}
   > Set Text(Text="To Text (String)"(InString=@n1), self=Get Text_Secondary() -> (Text_Secondary:Common Text Object Reference))
  ]}
]
variables [PresenceTextColors {type="map<LinearColor>", default {Default=<rgba 1 1 1 0.85>, InGame=<rgba 0.0353 0.9804 0.2275 1>}}]
widgets {
 info {inheritedWidgetCount=0, parentClass=&UserWidget, rootWidgetClass=&VerticalBox, widgetCount=10}
 widgets [
  VerticalBox_SocialInfo {widgetClassPath=&VerticalBox, bIsVariable=true, widget=VerticalBox_SocialInfo}
  HBox_PrimaryInfo {
   widgetClassPath=&HorizontalBox
   parent=VerticalBox_SocialInfo
   widget=HBox_PrimaryInfo
   slot {class=VerticalBoxSlot, properties {horizontalAlignment=HAlign_Center, padding=<ltrb 0 0 0 4>}}
  }
  Text_Primary {
   widgetClassPath=&CommonTextBlock
   bIsVariable=true
   parent=HBox_PrimaryInfo
   widget=Text_Primary
   properties {
    applyLineHeightToBottomLine=false
    bIsScrollingEnabled=false
    justification=Center
    mobileFontSizeMultiplier=1.4
    text="[PH] {Primary Name}"
    font {
     bForceMonospaced=false
     bMaterialIsStencil=false
     fontMaterial=null
     fontObject=&HeadingNow
     letterSpacing=10
     monospacedWidth=1
     size=13.5
     skewAmount=0
     typefaceFontName=74 Regular
     outlineSettings {
      bApplyOutlineToDropShadows=false
      bMiteredCorners=false
      bSeparateFillAlpha=false
      outlineMaterial=null
      outlineSize=0
      outlineColor=<rgba 0 0 0 1>
     }
    }
    shadowOffset=<0 0>
    shapedTextOptions {bOverride_TextFlowDirection=true, bOverride_TextShapingMethod=false, textFlowDirection=Culture, textShapingMethod=Auto}
   }
   slot {class=HorizontalBoxSlot, properties {horizontalAlignment=HAlign_Center}}
  }
  Image_PlatformIcon_Primary {
   widgetClassPath=&Image
   bIsVariable=true
   parent=HBox_PrimaryInfo
   widget=Image_PlatformIcon_Primary
   properties {brush {
     bIsDynamicallyLoaded=false
     drawAs=Image
     imageType=NoImage
     mirroring=NoMirror
     resourceName=None
     resourceObject=null
     tiling=NoTile
     imageSize=<16 16>
     margin=<ltrb 0 0 0 0>
     outlineSettings {
      bUseBrushTransparency=false
      roundingType=HalfHeightRadius
      width=0
      color {colorUseRule=UseColor_Specified, specifiedColor=<rgba 0 0 0 0>}
      cornerRadii=<0 0 0 0>
     }
     tintColor {colorUseRule=UseColor_Specified, specifiedColor=<rgba 1 1 1 1>}
     uVRegion {bIsValid=false, max=<0 0>, min=<0 0>}
    }}
   slot {class=HorizontalBoxSlot, properties {horizontalAlignment=HAlign_Left, verticalAlignment=VAlign_Center, padding=<ltrb 4 0 0 0>}}
  }
  HBox_SecondaryInfo {
   widgetClassPath=&HorizontalBox
   bIsVariable=true
   parent=VerticalBox_SocialInfo
   widget=HBox_SecondaryInfo
   slot {class=VerticalBoxSlot, properties {horizontalAlignment=HAlign_Center}}
  }
  Text_Secondary {
   widgetClassPath=&CommonTextBlock
   bIsVariable=true
   parent=HBox_SecondaryInfo
   widget=Text_Secondary
   properties {
    applyLineHeightToBottomLine=false
    bIsScrollingEnabled=false
    clipping=ClipToBounds
    justification=Center
    mobileFontSizeMultiplier=1.6
    text="(PH) {Platform}"
    textOverflowPolicy=Ellipsis
    colorAndOpacity {colorUseRule=UseColor_Specified, specifiedColor=<rgba 0.75 0.75 0.75 1>}
    font {
     bForceMonospaced=false
     bMaterialIsStencil=false
     fontMaterial=null
     fontObject=&HeadingNow
     letterSpacing=0
     monospacedWidth=1
     size=18
     skewAmount=0
     typefaceFontName=74 Regular
     outlineSettings {
      bApplyOutlineToDropShadows=false
      bMiteredCorners=false
      bSeparateFillAlpha=false
      outlineMaterial=null
      outlineSize=0
      outlineColor=<rgba 0 0 0 1>
     }
    }
    shadowOffset=<0 0>
    shapedTextOptions {bOverride_TextFlowDirection=true, bOverride_TextShapingMethod=false, textFlowDirection=Culture, textShapingMethod=Auto}
   }
   slot {class=HorizontalBoxSlot, properties {horizontalAlignment=HAlign_Center}}
  }
  Image_PlatformIcon_Secondary {
   widgetClassPath=&Image
   bIsVariable=true
   parent=HBox_SecondaryInfo
   widget=Image_PlatformIcon_Secondary
   properties {brush {
     bIsDynamicallyLoaded=false
     drawAs=Image
     imageType=NoImage
     mirroring=NoMirror
     resourceName=None
     resourceObject=null
     tiling=NoTile
     imageSize=<16 16>
     margin=<ltrb 0 0 0 0>
     outlineSettings {
      bUseBrushTransparency=false
      roundingType=HalfHeightRadius
      width=0
      color {colorUseRule=UseColor_Specified, specifiedColor=<rgba 0 0 0 0>}
      cornerRadii=<0 0 0 0>
     }
     tintColor {colorUseRule=UseColor_Specified, specifiedColor=<rgba 1 1 1 1>}
     uVRegion {bIsValid=false, max=<0 0>, min=<0 0>}
    }}
   slot {class=HorizontalBoxSlot, properties {horizontalAlignment=HAlign_Left, verticalAlignment=VAlign_Center, padding=<ltrb 4 0 0 0>}}
  }
  SizeBox_InGameStatus {
   widgetClassPath=&SizeBox
   bIsVariable=true
   parent=VerticalBox_SocialInfo
   widget=SizeBox_InGameStatus
   properties {bOverride_MaxDesiredWidth=true, maxDesiredWidth=100}
   slot {class=VerticalBoxSlot}
  }
  Text_InGameStatus {
   widgetClassPath=&CommonTextBlock
   bIsVariable=true
   parent=SizeBox_InGameStatus
   widget=Text_InGameStatus
   properties {
    applyLineHeightToBottomLine=false
    bIsScrollingEnabled=false
    clipping=ClipToBounds
    mobileFontSizeMultiplier=1.6
    text="(PH) {Status} - {Game Name}"
    textOverflowPolicy=Ellipsis
    colorAndOpacity {colorUseRule=UseColor_Specified, specifiedColor=<rgba 0.0369 0.9823 0.227 1>}
    font {
     bForceMonospaced=false
     bMaterialIsStencil=false
     fontMaterial=null
     fontObject=&HeadingNow
     letterSpacing=0
     monospacedWidth=1
     size=15.75
     skewAmount=0
     typefaceFontName=74 Regular
     outlineSettings {
      bApplyOutlineToDropShadows=false
      bMiteredCorners=false
      bSeparateFillAlpha=false
      outlineMaterial=null
      outlineSize=0
      outlineColor=<rgba 0 0 0 1>
     }
    }
    shadowOffset=<0 0>
   }
   slot {class=SizeBoxSlot, properties {horizontalAlignment=HAlign_Center}}
  }
  Text_LastPlayed {
   widgetClassPath=&CommonTextBlock
   bIsVariable=true
   parent=VerticalBox_SocialInfo
   widget=Text_LastPlayed
   properties {
    applyLineHeightToBottomLine=false
    bIsScrollingEnabled=false
    clipping=ClipToBounds
    justification=Center
    mobileFontSizeMultiplier=1.6
    text="(PH) Partied 24h ago"
    textOverflowPolicy=Ellipsis
    visibility=Collapsed
    colorAndOpacity {colorUseRule=UseColor_Specified, specifiedColor=<rgba 0.75 0.75 0.75 1>}
    font {
     bForceMonospaced=false
     bMaterialIsStencil=false
     fontMaterial=null
     fontObject=&HeadingNow
     letterSpacing=0
     monospacedWidth=1
     size=18
     skewAmount=0
     typefaceFontName=74 Regular
     outlineSettings {
      bApplyOutlineToDropShadows=false
      bMiteredCorners=false
      bSeparateFillAlpha=false
      outlineMaterial=null
      outlineSize=0
      outlineColor=<rgba 0 0 0 1>
     }
    }
    shadowOffset=<0 0>
    shapedTextOptions {bOverride_TextFlowDirection=true, bOverride_TextShapingMethod=false, textFlowDirection=Culture, textShapingMethod=Auto}
   }
   slot {class=VerticalBoxSlot, properties {horizontalAlignment=HAlign_Center}}
  }
 ]
}
