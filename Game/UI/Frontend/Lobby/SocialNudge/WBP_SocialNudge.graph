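# Social nudge widget for a lobby team-member pedestal, showing avatar, social info and a party call-to-action (join, request, accept or invite) with a plus button.
# FortTeamMemberPedestalSocialNudgeViewModel, bound by member ID, supplies action, joining and invite state.
# Hover and up/down navigation set a focus state that, with that data, drives CTA, visibility and outlines.
# Clicks open the sidebar when no action exists, toggle a context menu on touch, or perform the action.
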WidgetBlueprint /Game/UI/Frontend/Lobby/SocialNudge/WBP_SocialNudge
animations [FadeAnimation {
  binding_count=1
  length_seconds=0.25
  object=&self:FadeAnimation
  sequence {bindings [Button_SocialNudge {
     id=#1
     possessed_class=PanelButton_C
     tracks [
      {
       type=MovieSceneFloatTrack
       display_name=Render Opacity
       sections [{type=MovieSceneFloatSection, end_frame=5, start_frame=0, channels [None {type=MovieSceneScriptingFloatChannel, keys [[0,0],[5,1]]}]}]
      }
      {
       type=MovieScene2DTransformTrack
       sections [{
         type=MovieScene2DTransformSection
         end_frame=5
         start_frame=0
         channels [Translation.Y {type=MovieSceneScriptingFloatChannel, keys [[0,20],[5,0]]}]
        }]
      }
     ]
    }]}
 }]
graphs [
 EventGraph {nodes [
   "On Hovered (Button_SocialNudge)"!() -> (OutputDelegate, then, Button:Common Button Base Object Reference) {is=title K2Node_ComponentBoundEvent}
   > SetFocusAndUpdateSocialNudgeState!(execute, self, UseAutoFocus:bool=false, CurrentFocusedWidget:SocialNudgeFocusState Enum=NewEnumerator1, IsHovered:bool=true) -> (then)
   n1 = Get Text_InGameStatus!(self:WBP Social Nudge Info Object Reference=Get WBP_SocialNudgeInfo() -> (WBP_SocialNudgeInfo:WBP Social Nudge Info Object Reference)) -> (Text_InGameStatus:Common Text Object Reference)
   > SetScrollingEnabled!(execute, self:Common Text Object Reference=@n1, bInIsScrollingEnabled:bool=true) -> (then)
   "On Unhovered (Button_SocialNudge)"!() -> (OutputDelegate, then, Button:Common Button Base Object Reference) {is=title K2Node_ComponentBoundEvent}
   > SetFocusAndUpdateSocialNudgeState(UseAutoFocus=false, CurrentFocusedWidget=NewEnumerator1, IsHovered=false)
   > SetScrollingEnabled(self=@n1, bInIsScrollingEnabled=Select(Option 0:bool=false, Option 1:bool=true, Index=IsMobileGame()) -> (ReturnValue:bool))
   > ResetScrollState(self:Common Text Object Reference=@n1)
   "On Clicked (Button_Plus)"!() -> (OutputDelegate, then, Button:Common Button Base Object Reference) {is=title K2Node_ComponentBoundEvent}
   > OpenSidebarOrPerformAction()
   "On Hovered (Button_Plus)"!() -> (OutputDelegate, then, Button:Common Button Base Object Reference) {is=title K2Node_ComponentBoundEvent}
   > SetFocusAndUpdateSocialNudgeState(UseAutoFocus=false, CurrentFocusedWidget=NewEnumerator2, IsHovered=true)
   "On Unhovered (Button_Plus)"!() -> (OutputDelegate, then, Button:Common Button Base Object Reference) {is=title K2Node_ComponentBoundEvent}
   > SetFocusAndUpdateSocialNudgeState(UseAutoFocus=false, CurrentFocusedWidget=NewEnumerator2, IsHovered=false)
   "On Clicked (Button_SocialNudge)"!() -> (OutputDelegate, then, Button:Common Button Base Object Reference) {is=title K2Node_ComponentBoundEvent}
   > OpenSidebarOrPerformAction()
   Event PreConstruct()
   > SetCompactUIStyle() {is=title K2Node_CallFunction}
   Event Construct()
   > UpdateSocialNudgeViewModelMemberId()
   > Set PlatformTagDisplayRule!(
    execute
    PlatformTagDisplayRule:EFortSocialPlatformTagDisplayRule Enum=Always
    self:Fort Social User Name Viewmodel Object Reference=Get NameInfo!(self:Fort Team Member Pedestal Social Nudge View Model Object Reference=Get FortTeamMemberPedestalSocialNudgeViewModel() -> (FortTeamMemberPedestalSocialNudgeViewModel:Fort Team Member Pedestal Social Nudge View Model Object Reference)) -> (NameInfo:Fort Social User Name Viewmodel Object Reference)
   ) -> (then, Output_Get:EFortSocialPlatformTagDisplayRule Enum=Never)
   > PlayAnimation(InAnimation=Get FadeAnimation() -> (FadeAnimation:Widget Animation Object Reference))
   Event Destruct()
   > PlayAnimation(InAnimation=Get FadeAnimation(), PlayMode=Reverse)
   Event OnInitialized()
   > n2 = Construct Object from Class(Class=/Script/FortniteUI.UIKitContextMenuActionContext) -> (ReturnValue:UIKit Context Menu Action Context Object Reference)
   > Set ContextMenuActionContext(ContextMenuActionContext:UIKit Context Menu Action Context Object Reference=@n2) -> (Output_Get:UIKit Context Menu Action Context Object Reference)
   > SetContext!(
    execute
    self:WBP UIKit Context Menu Anchor Button Base Object Reference=Get WBP_UIKit_ContextMenuAnchorButton() -> (WBP_UIKit_ContextMenuAnchorButton:WBP UIKit Context Menu Anchor Button Object Reference)
    InteractionContext_Data:UIKit Context Menu Action Context Object Reference=Get ContextMenuActionContext() -> (ContextMenuActionContext:UIKit Context Menu Action Context Object Reference)
   ) -> (then)
  ]}
 GetCurrentAvailableActionFromViewModel {nodes [
   GetCurrentAvailableActionFromViewModel() {is=title K2Node_FunctionEntry}
   n1 = Get FortTeamMemberPedestalSocialNudgeViewModel()
   > Is Valid(InputObject=@n1)
   > Return Node!(
    execute
    CurrentAvailableAction:EPedestalSocialNudgeAction Enum=Get CurrentAvailableAction!(self:Fort Team Member Pedestal Social Nudge View Model Object Reference=@n1) -> (CurrentAvailableAction:EPedestalSocialNudgeAction Enum=None)
   )
  ]}
 GetPresenceStatus {nodes [
   n1 = GetPresenceStatus!() -> (then, IsInZone:bool=~) {is=title K2Node_FunctionEntry}
   > n2 = InZonePresenceEnabled!(
    execute
    self:Fort Social User Info View Model Object Reference=Get UserInfo!(self:Fort Team Member Pedestal Social Nudge View Model Object Reference=Get FortTeamMemberPedestalSocialNudgeViewModel()) -> (UserInfo:Fort Social User Info View Model Object Reference)
   ) -> (then, ReturnValue:bool)
   > Return Node!(execute, IsInGame:bool=AND Boolean(A=@n2, B=@n1))
  ]}
 GetPrimaryName {nodes [
   n1 = FunctionEntry!() -> (then, Nickname:String, DisplayName:String) {is=class K2Node_FunctionEntry, also {title=""}}
   > Return Node!(
    execute
    PrimaryName:String=Select(Option 0:String=@n1, Option 1:String=@n1.DisplayName, Index=IsEmpty(InString=@n1)) -> (ReturnValue:String)
   )
  ]}
 HandlePedestalAvailableActionChanged {nodes [
   n1 = HandlePedestalAvailableActionChanged!() -> (then, AvailableAction:EPedestalSocialNudgeAction Enum=None) {is=title K2Node_FunctionEntry}
   > UpdateSocialNudgeState!(execute, self, CurrentAvailableAction:EPedestalSocialNudgeAction Enum=@n1) -> (then)
   > UpdateContextualMenuActions!(execute, self, AvailableAction:EPedestalSocialNudgeAction Enum=@n1) -> (then)
  ]}
 IsSocialPanelVisible {nodes [
   IsSocialPanelVisible() {is=title K2Node_FunctionEntry}
   > n1 = Branch(Condition=IsVisible(self=Get Button_SocialNudge() -> (Button_SocialNudge:Panel Button Object Reference)))
   n2 = Get WidgetSwitcher_SocialNudgePlus() -> (WidgetSwitcher_SocialNudgePlus:Widget Switcher Object Reference)
   > n3 = Branch(Condition="Equal (Object)"(
     A=GetActiveWidget!(self:Widget Switcher Object Reference=@n2) -> (ReturnValue:Widget Object Reference)
     B=Get GridPanel_SocialNudge() -> (GridPanel_SocialNudge:Grid Panel Object Reference)
    ))
   > Return Node!(execute, Visible:bool=IsVisible(self=Get GridPanel_SocialNudge()))
   Return Node(execute=[@n3.else, @n1.else], Visible=false)
  ]}
 NavigateOut {nodes [
   NavigateOut() {is=title K2Node_FunctionEntry}
   > SetFocusAndUpdateSocialNudgeState(
    UseAutoFocus=false
    CurrentFocusedWidget=Get CurrentFocusedWidget() -> (CurrentFocusedWidget:SocialNudgeFocusState Enum=NewEnumerator0)
    IsHovered=false
   )
  ]}
 NavigationSocialNudgeAndPlusButtons {nodes [
   n1 = NavigationSocialNudgeAndPlusButtons!() -> (then, Navigation:EUINavigation Enum) {is=title K2Node_FunctionEntry}
   > n2 = Switch on EUINavigation!(execute, Selection:EUINavigation Enum=@n1) -> (Left:Exec, Right:Exec, Up:Exec, Down:Exec, Next:Exec, Previous:Exec, Invalid:Exec) {is=title K2Node_SwitchEnum}
   n3 = Get FortTeamMemberPedestalSocialNudgeViewModel()
   n4 = Branch(
    execute=@n2.Up
    Condition="Not Equal (Enum)"!(
     A:EPedestalSocialNudgeAction Enum=Get CurrentAvailableAction(self=@n3) -> (CurrentAvailableAction=None)
     B:EPedestalSocialNudgeAction Enum=None
    ) -> (ReturnValue:bool)
   )
   n5 = Get WBP_PlayerPad_AddPlayer() -> (WBP_PlayerPad_AddPlayer:WBP Player Pad Add Player Object Reference)
   > OnTransitionUnfocus(self:WBP Player Pad Add Player Object Reference=@n5)
   n6 = Get Block_Outline() -> (Block_Outline:WBP UIKit Block Outline Object Reference)
   > RequestTransition!(execute, self:UIKit Block Object Reference=@n6, State:EUIKitBlockVisualState Enum=Focused) -> (then)
   > SetFocusAndUpdateSocialNudgeState(UseAutoFocus=false, CurrentFocusedWidget=NewEnumerator2, IsHovered=false)
   > SetFocusAndUpdateSocialNudgeState(UseAutoFocus=false, CurrentFocusedWidget=NewEnumerator1, IsHovered=true)
   > Return Node!(execute, ReturnValue:Widget Object Reference=Get Button_SocialNudge())
   Return Node(execute=[@n7.then, @n4.else], ReturnValue=Get Button_Plus() -> (Button_Plus:Panel Button Object Reference))
   OnTransitionFocus(execute=@n2.Down, self:WBP Player Pad Add Player Object Reference=@n5)
   > RequestTransition(self=@n6, State=Unfocused)
   > SpawnSound2D!(
    execute
    Sound:Sound Base Object Reference=/Game/Sounds/Fort_UI_Sounds/FrontEnd/DefaultSounds/PressedButtons/MSS_UI_Button_Hover_Default
    VolumeMultiplier:float=1
    PitchMultiplier:float=1
    StartTime:float=0
    ConcurrencySettings:Sound Concurrency Object Reference
    bPersistAcrossLevelTransition:bool=false
    bAutoDestroy:bool=true
   ) -> (then, ReturnValue:Audio Component Object Reference)
   > SetFocusAndUpdateSocialNudgeState(UseAutoFocus=false, CurrentFocusedWidget=NewEnumerator1, IsHovered=false)
   > n7 = SetFocusAndUpdateSocialNudgeState(UseAutoFocus=false, CurrentFocusedWidget=NewEnumerator2, IsHovered=true)
  ]}
 OpenSidebarOrPerformAction {nodes [
   OpenSidebarOrPerformAction() {is=title K2Node_FunctionEntry}
   n1 = Get FortTeamMemberPedestalSocialNudgeViewModel()
   > Is Valid(InputObject=@n1)
   > n2 = Branch(Condition=Get bIsJoining!(self:Fort Team Member Pedestal Social Nudge View Model Object Reference=@n1) -> (bIsJoining:bool))
   n3 = Get CurrentAvailableAction(self=@n1) -> (CurrentAvailableAction=None)
   n4 = Branch(execute=@n2.else, Condition="Equal (Enum)"(A:EPedestalSocialNudgeAction Enum=@n3, B:EPedestalSocialNudgeAction Enum=None))
   > OpenSidebar(self:Fort Team Member Pedestal Social Nudge View Model Object Reference=Get FortTeamMemberPedestalSocialNudgeViewModel())
   n5 = Branch(execute=@n4.else, Condition=IsUsingTouch(PlayerController=GetOwningPlayer()))
   > ToggleContextMenu(self:WBP UIKit Context Menu Anchor Button Base Object Reference=Get WBP_UIKit_ContextMenuAnchorButton())
   n6 = Get CurrentFocusedWidget() -> (CurrentFocusedWidget=NewEnumerator0)
   PerformAction!(
    execute=@n5.else
    self:Fort Team Member Pedestal Social Nudge View Model Object Reference=@n1
    Action:EPedestalSocialNudgeAction Enum=Select(
     Option 0:EPedestalSocialNudgeAction Enum=@n3
     Option 1:EPedestalSocialNudgeAction Enum=Invite
     Index="Equal (Enum)"(A:SocialNudgeFocusState Enum=@n6, B:SocialNudgeFocusState Enum=NewEnumerator2)
    ) -> (ReturnValue:EPedestalSocialNudgeAction Enum)
   ) -> (then)
  ]}
 PerformContextualMenuAction {nodes [
   n1 = PerformContextualMenuAction!() -> (then, InteractionNameID:Name) {is=title K2Node_FunctionEntry}
   n2 = Get InviteToParty_InteractionId() -> (InviteToParty_InteractionId:Name=None)
   n3 = Get FortTeamMemberPedestalSocialNudgeViewModel()
   > PerformAction(
    self=@n3
    Action=Select(Option 0=Get CurrentAvailableAction(self=@n3) -> (CurrentAvailableAction=None), Option 1=Invite, Index="Equal (Name)"(A=@n1, B=@n2))
   )
   > CloseMenu(self:WBP UIKit Context Menu Anchor Button Base Object Reference=Get WBP_UIKit_ContextMenuAnchorButton())
  ]}
 SetCompactUIStyle {nodes [
   SetCompactUIStyle() {is=title K2Node_FunctionEntry}
   > Branch(Condition=IsMobileGame())
   > SetCornerRadius!(
    execute
    self:WBP UIKit Backplate Base Object Reference=Get UIKit_Backplate() -> (UIKit_Backplate:WBP UIKit Backplate Object Reference)
    Corner Radius:E_UI_BackplateCornerRadius Enum=NewEnumerator6
   ) -> (then)
   n1 = Get WBP_SocialNudgeInfo()
   > SetVerticalAlignment(
    self:Horizontal Box Slot Object Reference=[
     SlotAsHorizontalBoxSlot!(Widget:Widget Object Reference=Get Text_Primary!(self:WBP Social Nudge Info Object Reference=@n1) -> (Text_Primary:Common Text Object Reference)) -> (ReturnValue:Horizontal Box Slot Object Reference)
     SlotAsHorizontalBoxSlot(Widget=Get Text_Secondary!(self:WBP Social Nudge Info Object Reference=@n1) -> (Text_Secondary:Common Text Object Reference))
    ]
    InVerticalAlignment=VAlign_Center
   )
   n2 = Get Image_PlatformIcon_Primary!(self:WBP Social Nudge Info Object Reference=@n1) -> (Image_PlatformIcon_Primary:Image Object Reference)
   n3 = Get!(self:Image Object Reference=@n2) -> (
    Brush_TintColor:Slate Color Structure
    Brush_DrawAs:ESlateBrushDrawType Enum
    Brush_Tiling:ESlateBrushTileType Enum
    Brush_Mirroring:ESlateBrushMirrorType Enum
    Brush_ImageSize:Vector 2D Structure
    Brush_Margin:Margin Structure
    Brush_ResourceObject:Object Reference
    Brush_OutlineSettings:Slate Brush Outline Settings Structure
   )
   > Set!(
    execute
    Brush_TintColor:Slate Color Structure=@n3
    Brush_DrawAs:ESlateBrushDrawType Enum=@n3.Brush_DrawAs
    Brush_Tiling:ESlateBrushTileType Enum=@n3.Brush_Tiling
    Brush_Mirroring:ESlateBrushMirrorType Enum=@n3.Brush_Mirroring
    Brush_ImageSize_X:double=32
    Brush_ImageSize_Y:double=32
    Brush_Margin:Margin Structure=@n3.Brush_Margin
    Brush_ResourceObject:Object Reference=@n3.Brush_ResourceObject
    Brush_OutlineSettings:Slate Brush Outline Settings Structure=@n3.Brush_OutlineSettings
    self:Image Object Reference=@n2
   ) -> (then, Output_Get:Slate Brush Structure)
   n4 = Get Image_PlatformIcon_Secondary!(self:WBP Social Nudge Info Object Reference=@n1) -> (Image_PlatformIcon_Secondary:Image Object Reference)
   n5 = Get(self=@n4)
   > Set(
    Brush_TintColor=@n5
    Brush_DrawAs=@n5.Brush_DrawAs
    Brush_Tiling=@n5.Brush_Tiling
    Brush_Mirroring=@n5.Brush_Mirroring
    Brush_ImageSize_X=32
    Brush_ImageSize_Y=32
    Brush_Margin=@n5.Brush_Margin
    Brush_ResourceObject=@n5.Brush_ResourceObject
    Brush_OutlineSettings=@n5.Brush_OutlineSettings
    self=@n4
   )
   > SetScrollingEnabled(self=Get Text_InGameStatus(self=Get WBP_SocialNudgeInfo()), bInIsScrollingEnabled=true)
  ]}
 SetFocusAndUpdateSocialNudgeState {nodes [
   n1 = SetFocusAndUpdateSocialNudgeState!() -> (then, UseAutoFocus:bool=~, CurrentFocusedWidget:SocialNudgeFocusState Enum=NewEnumerator0, IsHovered:bool=~) {is=title K2Node_FunctionEntry}
   > n2 = GetCurrentAvailableActionFromViewModel!(execute, self) -> (then, CurrentAvailableAction:EPedestalSocialNudgeAction Enum=None) {is=title K2Node_CallFunction}
   K2Node_Composite_1 = Select focused widget based on the current available action!(Index:EPedestalSocialNudgeAction Enum=@n2) -> (ReturnValue:SocialNudgeFocusState Enum=NewEnumerator0) {is=title K2Node_Composite}
   > n3 = Set CurrentFocusedWidget(CurrentFocusedWidget:SocialNudgeFocusState Enum=Select(
     Option 0:SocialNudgeFocusState Enum=Select(Option 0=NewEnumerator0, Option 1=@n1.CurrentFocusedWidget, Index=@n1.IsHovered)
     Option 1:SocialNudgeFocusState Enum=@K2Node_Composite_1
     Index=Get UseAutoFocus() -> (UseAutoFocus:bool)
    ) -> (ReturnValue:SocialNudgeFocusState Enum)) -> (Output_Get:SocialNudgeFocusState Enum=NewEnumerator0)
   > SetFocus(self=Select!(
     NewEnumerator0:Widget Object Reference
     NewEnumerator1:Widget Object Reference=Get Button_SocialNudge()
     NewEnumerator2:Widget Object Reference=Get Button_Plus()
     Index:SocialNudgeFocusState Enum=@n3
    ) -> (ReturnValue:Widget Object Reference))
   > UpdateSocialNudgeState(CurrentAvailableAction=@n2)
   > n4 = Switch on SocialNudgeFocusState!(execute, Selection:SocialNudgeFocusState Enum=Get CurrentFocusedWidget() -> (CurrentFocusedWidget=NewEnumerator0)) -> (NewEnumerator0:Exec, NewEnumerator1:Exec, NewEnumerator2:Exec) {is=title K2Node_SwitchEnum}
   n5 = Get IsHovered() -> (IsHovered:bool)
   RequestTransition(
    execute=@n4.NewEnumerator1
    self=Get Block_Outline()
    State=Select!(Option 0:EUIKitBlockVisualState Enum=Unfocused, Option 1:EUIKitBlockVisualState Enum=Focused, Index:bool=@n5) -> (ReturnValue:EUIKitBlockVisualState Enum)
   )
   n6 = Branch(execute=@n4.NewEnumerator2, Condition=@n5)
   n7 = Get WBP_PlayerPad_AddPlayer()
   > OnTransitionFocus(self=@n7)
   OnTransitionUnfocus(execute=@n6.else, self=@n7)
  ]}
 Select focused widget based on the current available action {
  path=SetFocusAndUpdateSocialNudgeState.K2Node_Composite_1.Select focused widget based on the current available action
  nodes [Outputs!(ReturnValue:SocialNudgeFocusState Enum=Select!(
     None:SocialNudgeFocusState Enum=NewEnumerator2
     Join:SocialNudgeFocusState Enum=NewEnumerator1
     RequestToJoin:SocialNudgeFocusState Enum=NewEnumerator1
     Invite:SocialNudgeFocusState Enum=NewEnumerator2
     AcceptInvite:SocialNudgeFocusState Enum=NewEnumerator1
     Accept1v1Invite:SocialNudgeFocusState Enum=NewEnumerator1
     Index:EPedestalSocialNudgeAction Enum=Inputs!() -> (Index:EPedestalSocialNudgeAction Enum=None)
    ) -> (ReturnValue:SocialNudgeFocusState Enum))]
 }
 SetMemberUniqueId {nodes [
   n1 = SetMemberUniqueId!() -> (then, MemberUniqueId:Unique Net Id Structure) {is=title K2Node_FunctionEntry}
   > Set MemberUniqueId(MemberUniqueId:Unique Net Id Structure=@n1) -> (Output_Get:Unique Net Id Structure)
   > UpdateSocialNudgeViewModelMemberId()
  ]}
 UpdateContextualMenuActions {nodes [
   UpdateContextualMenuActions!() -> (then, AvailableAction:EPedestalSocialNudgeAction Enum=None) {is=title K2Node_FunctionEntry}
   > CloseMenu(self=Get WBP_UIKit_ContextMenuAnchorButton())
   n1 = Get ContextMenuActionContext()
   > RemoveAllInteractions(self:UIKit Context Menu Action Context Object Reference=@n1)
   > AddInteraction!(
    execute
    self:UIKit Context Menu Action Context Object Reference=@n1
    InActionInfo_DisplayName:Text="LOCTEXT(\"Invite to Party\")"
    InActionInfo_IconTexture:Texture 2D Soft Object Reference
    InActionInfo_IconMaterial:Material Instance Soft Object Reference
    InActionInfo_SortPriority:int=1
    InActionInfo_bIsAvailable:bool=True
    InActionInfo_bIsEnabled:bool=True
    InActionInfo_bIsSelected:bool=False
    InActionInfo_bAddDivider:bool=False
    InActionInfo_bPositionDividerAfter:bool=False
    InActionInfo_bCloseMenuOnSelect:bool=False
    InActionInfo_InteractionNameID:Name=Get InviteToParty_InteractionId() -> (InviteToParty_InteractionId=None)
    InActionInfo_Interaction:Data Table Row Handle Structure=(DataTable=None,RowName="")
    InActionInfo_PressedDelegate:Delegate="(null).None"
    InActionInfo_ReleasedDelegate:Delegate="(null).None"
    InInteractionDelegate:"Delegate (by ref)"=Create Event()
   ) -> (then)
   n2 = Get AvailableAction() -> (AvailableAction:EPedestalSocialNudgeAction Enum=None)
   > Branch(Condition=AND Boolean(A="Not Equal (Enum)"(A=@n2, B=None), B="Not Equal (Enum)"(A=@n2, B=Invite)))
   n3 = Get AvailableAction() -> (AvailableAction=None)
   > AddInteraction(
    self=@n1
    InActionInfo_DisplayName=Select(
     None:Text
     Join:Text=Get JoinParty() -> (JoinParty:Text)
     RequestToJoin:Text=Get RequestJoin() -> (RequestJoin:Text)
     Invite:Text
     AcceptInvite:Text=Get AcceptInvite() -> (AcceptInvite:Text)
     Accept1v1Invite:Text=Get Accept1v1Invite() -> (Accept1v1Invite:Text)
     Index=@n3
    ) -> (ReturnValue:Text)
    InActionInfo_SortPriority=Select(None:int=2, Join:int=2, RequestToJoin:int=2, Invite:int=2, AcceptInvite:int=0, Accept1v1Invite:int=0, Index=@n3) -> (ReturnValue:int)
    InActionInfo_bIsAvailable=True
    InActionInfo_bIsEnabled=True
    InActionInfo_bIsSelected=False
    InActionInfo_bAddDivider=False
    InActionInfo_bPositionDividerAfter=False
    InActionInfo_bCloseMenuOnSelect=False
    InActionInfo_InteractionNameID=None
    InActionInfo_Interaction=(DataTable=None,RowName="")
    InActionInfo_PressedDelegate="(null).None"
    InActionInfo_ReleasedDelegate="(null).None"
    InInteractionDelegate=Create Event()
   )
  ]}
 UpdateSocialNudgeState {nodes [
   UpdateSocialNudgeState!() -> (then, CurrentAvailableAction:EPedestalSocialNudgeAction Enum=None) {is=title K2Node_FunctionEntry}
   n1 = Get FortTeamMemberPedestalSocialNudgeViewModel()
   > Is Valid(InputObject=@n1)
   > Set IsJoining(IsJoining:bool=Get bIsJoining(self=@n1)) -> (Output_Get:bool)
   n2 = Get bIsInviteInProgress!(self:Fort Team Member Pedestal Social Nudge View Model Object Reference=@n1) -> (bIsInviteInProgress:bool)
   > SetVisibility(
    self=Get FX_InviteInProgress() -> (FX_InviteInProgress:Image Object Reference)
    InVisibility=Select!(Option 0:ESlateVisibility Enum=Collapsed, Option 1:ESlateVisibility Enum=SelfHitTestInvisible, Index:bool=@n2) -> (ReturnValue:ESlateVisibility Enum)
   )
   > n3 = Sequence()
   n4 = SelectText(A=Get InviteSent() -> (InviteSent:Text), B=Get InviteToParty() -> (InviteToParty:Text), bPickA=@n2)
   n5 = "Equal (Enum)"(A=Get CurrentFocusedWidget() -> (CurrentFocusedWidget=NewEnumerator0), B=NewEnumerator2)
   n6 = SelectText(A=@n4, B=Get JoinParty(), bPickA=@n5)
   n7 = SelectText(A=@n4, B=Get AcceptInvite(), bPickA=@n5)
   n8 = SelectText(A=@n4, B=Get RequestJoin(), bPickA=@n5)
   n9 = Get Accept1v1Invite()
   n10 = Get CurrentAvailableAction!() -> (CurrentAvailableAction=None)
   n11 = Get IsJoining() -> (IsJoining:bool)
   n12 = Select(Option 0:E_UI_HighlightType Enum=NewEnumerator1, Option 1:E_UI_HighlightType Enum=NewEnumerator5, Index=@n5) -> (ReturnValue:E_UI_HighlightType Enum)
   UpdateTypeAndTextOfCTA!(
    execute=@n3.then_1
    self
    HighlightType:E_UI_HighlightType Enum=Select(
     Option 0=Select!(
      None:E_UI_HighlightType Enum=NewEnumerator5
      Join:E_UI_HighlightType Enum=NewEnumerator5
      RequestToJoin:E_UI_HighlightType Enum=NewEnumerator5
      Invite:E_UI_HighlightType Enum=NewEnumerator5
      AcceptInvite:E_UI_HighlightType Enum=@n12
      Accept1v1Invite:E_UI_HighlightType Enum=NewEnumerator1
      Index:EPedestalSocialNudgeAction Enum=@n10
     ) -> (ReturnValue)
     Option 1=NewEnumerator1
     Index=@n11
    )
    Text:Text=SelectText(
     A=Get JoiningParty() -> (JoiningParty:Text)
     B=Select(None:Text, Join:Text=@n6, RequestToJoin:Text=@n8, Invite:Text=@n4, AcceptInvite:Text=@n7, Accept1v1Invite:Text=@n9, Index=@n10) -> (ReturnValue:Text)
     bPickA=@n11
    )
   ) -> (then)
   n13 = NOT Boolean(A=Get IsJoining())
   n14 = "Equal (Enum)"(A:EPedestalSocialNudgeAction Enum=Get CurrentAvailableAction() -> (CurrentAvailableAction=None), B:EPedestalSocialNudgeAction Enum=None)
   > SetActiveWidget(
    self:Widget Switcher Object Reference=Get WidgetSwitcher_SocialNudgePlus()
    Widget=Select!(
     Option 0:Widget Object Reference=Get GridPanel_SocialNudge()
     Option 1:Widget Object Reference=Get Text_OpenSidebar() -> (Text_OpenSidebar:Common Text Object Reference)
     Index:bool=AND Boolean(A=@n14, B=@n13)
    ) -> (ReturnValue:Widget Object Reference)
   )
   > Broadcast Field Value Changed!(execute, self:Widget Object Reference, FieldId:Field Notification Id Structure=(FieldName="IsSocialPanelVisible")) -> (then)
   n15 = "Equal (Enum)"(
    A:SocialNudgeFocusState Enum=Get CurrentFocusedWidget() -> (CurrentFocusedWidget=NewEnumerator0)
    B:SocialNudgeFocusState Enum=NewEnumerator0
   )
   n16 = Select(Option 0:ESlateVisibility Enum=SelfHitTestInvisible, Option 1:ESlateVisibility Enum=Collapsed, Index=@n15) -> (ReturnValue:ESlateVisibility Enum)
   n17 = Get CurrentAvailableAction() -> (CurrentAvailableAction=None)
   n18 = Get IsJoining()
   SetVisibility(
    execute=@n3.then_0
    self=Get Button_SocialNudge()
    InVisibility=Select(
     Option 0=Select!(
      None:ESlateVisibility Enum=@n16
      Join:ESlateVisibility Enum=SelfHitTestInvisible
      RequestToJoin:ESlateVisibility Enum=SelfHitTestInvisible
      Invite:ESlateVisibility Enum=SelfHitTestInvisible
      AcceptInvite:ESlateVisibility Enum=SelfHitTestInvisible
      Accept1v1Invite:ESlateVisibility Enum=SelfHitTestInvisible
      Index:EPedestalSocialNudgeAction Enum=@n17
     ) -> (ReturnValue)
     Option 1=SelfHitTestInvisible
     Index=@n18
    )
   )
   n19 = AND Boolean!(A="Not Equal (Enum)"(A=@n17, B=AcceptInvite), B=@n15, C:bool="Not Equal (Enum)"(A=@n17, B=Accept1v1Invite)) -> (ReturnValue)
   > SetRenderOpacity(
    self=Get CallToActions() -> (CallToActions:WBP UIKit Highlight Object Reference)
    InOpacity=Select!(Option 0:float=Select(Option 0=1, Option 1=0, Index=@n19), Option 1:float=1, Index:bool=@n18) -> (ReturnValue:float)
   )
   > SetRenderOpacity(
    self=[
     Get Button_Plus()
     Get Overlay-PlayerAvatar() -> (Overlay-PlayerAvatar:Overlay Object Reference)
     Get Overlay-SocialInfo() -> (Overlay-SocialInfo:Overlay Object Reference)
    ]
    InOpacity=Select(Option 0=1, Option 1=0.5, Index=Get IsJoining())
   )
  ]}
 UpdateSocialNudgeViewModelMemberId {nodes [
   UpdateSocialNudgeViewModelMemberId() {is=title K2Node_FunctionEntry}
   > SetSocialUserFromPlayerUniqueNetId!(
    execute
    self:Fort Team Member Pedestal Social Nudge View Model Object Reference=Get FortTeamMemberPedestalSocialNudgeViewModel()
    UniqueId:"Unique Net Id Structure (by ref)"=Get MemberUniqueId() -> (MemberUniqueId:Unique Net Id Structure)
   ) -> (then)
  ]}
 UpdateTypeAndTextOfCTA {nodes [
   n1 = UpdateTypeAndTextOfCTA!() -> (then, HighlightType:E_UI_HighlightType Enum=NewEnumerator0, Text:Text) {is=title K2Node_FunctionEntry}
   n2 = Get CallToActions()
   > SetType!(execute, self:WBP UIKit Highlight Base Object Reference=@n2, Type:E_UI_HighlightType Enum=@n1) -> (then)
   > SetText!(execute, self:WBP UIKit Highlight Base Object Reference=@n2, Text:Text=@n1.Text) -> (then)
  ]}
]
variables [
 Accept1v1Invite {type=string, default=Accept 1v1 Invite}
 AcceptInvite {type=string, default=Accept Invite}
 ContextMenuActionContext {type=UIKitContextMenuActionContext, default=null}
 CurrentFocusedWidget {type=SocialNudgeFocusState, default=None}
 InviteSent {type=string, default="Invite Sent!"}
 InviteToParty {type=string, default=Invite to Party}
 InviteToParty_InteractionId {type=string, default=InviteToParty}
 JoiningParty {type=string, default=Joining Party...}
 JoinParty {type=string, default=Join Party}
 MemberUniqueId {type=UniqueNetIdRepl, default=INVALID}
 RequestJoin {type=string, default=Request to Join}
 Test_CurrentAvailableAction {type=integer, default=-1}
]
widgets {
 info {
  inheritedWidgetCount=0
  namedSlotCount=2
  parentClass=&CommonActivatableWidget
  rootWidgetClass=&FortVisualAttachment
  widgetCount=22
 }
 widgets [
  FortVisualAttachment_0 {widgetClassPath=&FortVisualAttachment, widget=FortVisualAttachment_0, properties {contentAnchor=<0.5 1>}, slot=~}
  Overlay_1 {widgetClassPath=&Overlay, parent=FortVisualAttachment_0, widget=Overlay_1, slot {class=SizeBoxSlot}}
  WBP_CaptureForPostBufferUpdate {
   widgetClassPath=&WBP_CaptureForPostBufferUpdate
   bIsVariable=true
   parent=Overlay_1
   properties {renderTransform {angle=0, scale=<1 1.1>, shear=<0 0>, translation=<0 0>}}
   slot {class=OverlaySlot, properties {horizontalAlignment=HAlign_Fill, verticalAlignment=VAlign_Fill}}
  }
  VerticalBox_28 {
   widgetClassPath=&VerticalBox
   parent=Overlay_1
   widget=VerticalBox_28
   slot {class=OverlaySlot, properties {horizontalAlignment=HAlign_Fill, verticalAlignment=VAlign_Fill}}
  }
  Button_SocialNudge {
   widgetClassPath=/FortUILibrary/Foundation/Buttons/PanelButton
   bIsVariable=true
   parent=VerticalBox_28
   widget=Button_SocialNudge
   properties {
    bInteractableWhenSelected=true
    bSelectable=true
    bShouldUseFallbackDefaultInputAction=false
    bSimulateHoverOnTouchInput=false
    clickMethod=MouseDown
    navigation=&self:WidgetTree.Button_SocialNudge.WidgetNavigation_0
    onButtonBaseClicked=()
    onButtonBaseDoubleClicked=()
    onButtonBaseHovered=()
    onButtonBaseUnhovered=()
    onSelectedChangedBase=()
    style=/FortUILibrary/Foundation/Buttons/ButtonStyles/ButtonStyle-TextOnlyBase_Empty.ButtonStyle-TextOnlyBase_Empty_C
    hoveredSlateSoundOverride {resourceObject=/Game/Sounds/Fort_UI_Sounds/FrontEnd/New_FrontEnd/Fortnite_UI_Map_Hover_01_Cue}
   }
   slot {class=VerticalBoxSlot, properties {horizontalAlignment=HAlign_Left, verticalAlignment=VAlign_Top}}
  }
  WidgetSwitcher_SocialNudgePlus {widgetClassPath=&WidgetSwitcher, bIsVariable=true, namedSlotHost=Button_SocialNudge, widget=WidgetSwitcher_SocialNudgePlus}
  GridPanel_SocialNudge {
   widgetClassPath=&GridPanel
   bIsVariable=true
   parent=WidgetSwitcher_SocialNudgePlus
   widget=GridPanel_SocialNudge
   properties {columnFill [1], rowFill [0,0,0]}
   slot {class=WidgetSwitcherSlot}
  }
  CallToActions {
   widgetClassPath=/FortUILibrary/UIKit/Highlight/WBP_UIKit_Highlight
   bIsVariable=true
   parent=GridPanel_SocialNudge
   widget=CallToActions
   properties {highlightSize=Compact, highlightText="Invite to Party!", highlightType=Social}
   slot {class=GridSlot, properties {horizontalAlignment=HAlign_Center, verticalAlignment=VAlign_Top, padding=<ltrb 0 0 0 8>}}
  }
  Overlay-PlayerAvatar {
   widgetClassPath=&Overlay
   bIsVariable=true
   parent=GridPanel_SocialNudge
   widget=Overlay-PlayerAvatar
   slot {class=GridSlot, properties {horizontalAlignment=HAlign_Center, layer=10, row=1, verticalAlignment=VAlign_Top, nudge=<0 8>}}
  }
  FX_InviteInProgress {
   widgetClassPath=&Image
   bIsVariable=true
   parent=Overlay-PlayerAvatar
   widget=FX_InviteInProgress
   properties {brush {
     bIsDynamicallyLoaded=false
     drawAs=Image
     imageType=FullColor
     mirroring=NoMirror
     resourceName=None
     resourceObject=/Game/UI/Frontend/Lobby/SocialNudge/Materials/MI_UI_InvitePlayer
     tiling=NoTile
     imageSize=<32 32>
     margin=<ltrb 0 0 0 0>
     outlineSettings {
      bUseBrushTransparency=false
      roundingType=HalfHeightRadius
      width=0
      color {colorUseRule=UseColor_Specified, specifiedColor=<rgba 0 0 0 0>}
      cornerRadii=<0 0 0 0>
     }
     tintColor {colorUseRule=UseColor_Specified, specifiedColor=<rgba 1 1 1 1>}
     uVRegion {bIsValid=false, max=<0 0>, min=<0 0>}
    }}
   slot {class=OverlaySlot, properties {horizontalAlignment=HAlign_Fill, verticalAlignment=VAlign_Fill, padding=<ltrb -32 -32 -32 -32>}}
  }
  WBP_UIKit_Backplate {
   widgetClassPath=&WBP_UIKit_Backplate
   bIsVariable=true
   parent=Overlay-PlayerAvatar
   widget=WBP_UIKit_Backplate
   properties {brightness=Dark, cornerRadius=Circle, mobileCorners {backplate=25, button=0.3, circle=0.5, interactive Component=20, tab=24}}
   slot {class=OverlaySlot, properties {horizontalAlignment=HAlign_Fill, verticalAlignment=VAlign_Fill}}
  }
  WBP_PlayerAvatar {
   widgetClassPath=/Game/UI/Foundation/Profile/WBP_PlayerAvatar
   bIsVariable=true
   parent=Overlay-PlayerAvatar
   widget=WBP_PlayerAvatar
   properties {avatarSize=<51 51>}
   slot {class=OverlaySlot, properties {padding=<ltrb 5 5 5 5>}}
  }
  Overlay-SocialInfo {
   widgetClassPath=&Overlay
   bIsVariable=true
   parent=GridPanel_SocialNudge
   widget=Overlay-SocialInfo
   slot {class=GridSlot, properties {horizontalAlignment=HAlign_Center, layer=1, row=2, verticalAlignment=VAlign_Top}}
  }
  UIKit_Backplate {
   widgetClassPath=&WBP_UIKit_Backplate
   bIsVariable=true
   parent=Overlay-SocialInfo
   widget=UIKit_Backplate
   properties {brightness=Dark}
   slot {class=OverlaySlot, properties {horizontalAlignment=HAlign_Fill, verticalAlignment=VAlign_Fill}}
  }
  WBP_SocialNudgeInfo {
   widgetClassPath=&/WBP_SocialNudgeInfo
   bIsVariable=true
   parent=Overlay-SocialInfo
   widget=WBP_SocialNudgeInfo
   slot {class=OverlaySlot, properties {padding=<ltrb 16 12 16 12>}}
  }
  Block_Outline {
   widgetClassPath=/FortUILibrary/UIKit/Buttons/Blocks/WBP_UIKit_Block_Outline
   bIsVariable=true
   parent=Overlay-SocialInfo
   widget=Block_Outline
   properties {bottomLeftCornerRadius=22, bottomRightCornerRadius=22, radiusInPixels=true, topLeftCornerRadius=22, topRightCornerRadius=22}
   slot {class=OverlaySlot, properties {horizontalAlignment=HAlign_Fill, verticalAlignment=VAlign_Fill}}
  }
  Text_OpenSidebar {
   widgetClassPath=&CommonTextBlock
   bIsVariable=true
   parent=WidgetSwitcher_SocialNudgePlus
   widget=Text_OpenSidebar
   properties {
    applyLineHeightToBottomLine=false
    justification=Center
    text=Open Sidebar
    font {
     bForceMonospaced=false
     bMaterialIsStencil=false
     fontMaterial=null
     fontObject=&HeadingNow
     letterSpacing=0
     monospacedWidth=1
     size=18
     skewAmount=0
     typefaceFontName=75 Medium
     outlineSettings {
      bApplyOutlineToDropShadows=false
      bMiteredCorners=false
      bSeparateFillAlpha=false
      outlineMaterial=/FortUILibrary/MaterialInstances/Text/MI_UI_Block_Text_Outline
      outlineSize=2
      outlineColor=<rgba 0 0 0 1>
     }
    }
    shadowOffset=<0 0>
    shapedTextOptions {bOverride_TextFlowDirection=true, bOverride_TextShapingMethod=false, textFlowDirection=Culture, textShapingMethod=Auto}
   }
   slot {class=WidgetSwitcherSlot}
  }
  Button_Plus {
   widgetClassPath=&PanelButton
   bIsVariable=true
   parent=VerticalBox_28
   widget=Button_Plus
   properties {
    bInteractableWhenSelected=true
    bSelectable=true
    bShouldUseFallbackDefaultInputAction=false
    bSimulateHoverOnTouchInput=false
    clickMethod=MouseDown
    navigation=&self:WidgetTree.Button_Plus.WidgetNavigation_0
    onButtonBaseClicked=()
    onButtonBaseDoubleClicked=()
    onButtonBaseHovered=()
    onButtonBaseUnhovered=()
    onSelectedChangedBase=()
    style=&ButtonStyle-TextOnlyBase_Empty_C
    hoveredSlateSoundOverride {resourceObject=&Fortnite_UI_Map_Hover_01_Cue}
   }
   slot {class=VerticalBoxSlot, properties {horizontalAlignment=HAlign_Center, verticalAlignment=VAlign_Center, padding=<ltrb 0 16 0 0>}}
  }
  Overlay_0 {widgetClassPath=&Overlay, namedSlotHost=Button_Plus, widget=Overlay_0}
  WBP_PlayerPad_AddPlayer {
   widgetClassPath=/Game/UI/Frontend/Lobby/PlayerPad/WBP_PlayerPad_AddPlayer
   bIsVariable=true
   parent=Overlay_0
   widget=WBP_PlayerPad_AddPlayer
   slot {class=OverlaySlot}
  }
  CommonHardwareVisibilityBorder_0 {
   widgetClassPath=&CommonHardwareVisibilityBorder
   parent=Overlay_0
   widget=CommonHardwareVisibilityBorder_0
   properties {visibilityQuery {
     autoDescription=" ANY( Input.Touch )"
     tokenStreamVersion=0
     userDescription=Touch
     queryTokenStream [0,1,1,1,0]
     tagDictionary [{tagName=Input.Touch}]
    }}
   slot {class=OverlaySlot, properties {horizontalAlignment=HAlign_Center, verticalAlignment=VAlign_Center}}
  }
  WBP_UIKit_ContextMenuAnchorButton {
   widgetClassPath=/FortUILibrary/UIKit/ContextMenu/WBP_UIKit_ContextMenuAnchorButton
   bIsVariable=true
   parent=CommonHardwareVisibilityBorder_0
   widget=WBP_UIKit_ContextMenuAnchorButton
   properties {
    buttonClass=null
    buttonIcon=/FortUILibrary/Textures/Icons/T_UI_Icon_UIKit_ContextMenu_SDF
    buttonText=""
    menuPlacement=MenuPlacement_AboveRightAnchor
    button Padding=<ltrb 0 0 0 0>
   }
   slot {class=BorderSlot, properties {padding=<ltrb 0 0 0 0>}}
  }
 ]
}
