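# Lobby empty-slot "add player" widget (FortLobbyAddPlayer child) that switches between the legacy party-suggestion layout and WBP_SocialNudge via a runtime option.
# Visibility changes on the suggestion or social panel recompute bottom padding from NudgePosition and compact-UI spacing, offsetting nudges while shown.
# Pedestal hover drives empty-slot button selection, and login success or logout activates or deactivates the new nudge.
# Non-touch focus passes to the new nudge when enabled.
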WidgetBlueprint /Game/UI/Frontend/Lobby/SocialNudge/TeamMemberSocialNudgeWrapper
graphs [
 EventGraph {nodes [
   OnNudgeVisibilityChanged!() -> (OutputDelegate, then, InVisibility:ESlateVisibility Enum=Visible) {is=title K2Node_CustomEvent}
   > SetNudgePosition!(execute, self, Position:int=Get NudgePosition() -> (NudgePosition:int=0)) -> (then)
   OnSocialPanelVisibilityChanged!() -> (OutputDelegate, then, Object:Object Reference, Field:Field Notification Id Structure) {is=title K2Node_CustomEvent}
   > SetNudgePosition(Position=Get NudgePosition() -> (NudgePosition=0))
   Event Construct()
   > Add Field Value Changed Delegate!(
    execute
    self:Widget Object Reference=Get WBP_SocialNudge() -> (WBP_SocialNudge:WBP Social Nudge Object Reference)
    FieldId:Field Notification Id Structure=(FieldName="IsSocialPanelVisible")
    Delegate:Delegate=Create Event()
   ) -> (then)
   Event Destruct()
   > Remove Field Value Changed Delegate!(
    execute
    self:Widget Object Reference=Get WBP_SocialNudge()
    FieldId:Field Notification Id Structure=(FieldName="IsSocialPanelVisible")
    Delegate:Delegate=Create Event()
   ) -> (then)
   Event OnInitialized()
   n1 = Get bUseNewNameplatesAndSocialNudges!(self:Fort Runtime Options Object Reference=GetRuntimeOptions()) -> (bUseNewNameplatesAndSocialNudges:bool)
   > SetActiveWidget(
    self:Widget Switcher Object Reference=Get Switch_OldNewSocialNudges() -> (Switch_OldNewSocialNudges:Common Activatable Widget Switcher Object Reference)
    Widget=Select(
     Option 0:Widget Object Reference=Get FortHUDElementWrapper() -> (FortHUDElementWrapper:Fort HUDElement Wrapper Object Reference)
     Option 1:Widget Object Reference=Get WBP_SocialNudge()
     Index=@n1
    ) -> (ReturnValue:Widget Object Reference)
   )
   > Bind Event to On Visibility Changed(
    self:User Widget Object Reference=Get PartySuggestion() -> (PartySuggestion:Fort Party Suggestion Button Object Reference)
    Delegate=Create Event()
   ) {is=title K2Node_AddDelegate}
   n2 = Event HandleOwningPedestalIsHoveredChanged!() -> (OutputDelegate, then, bIsSelected:bool) {is=title K2Node_Event}
   > Branch(Condition=IsVisible())
   > SetIsSelected!(
    execute
    self:Common Button Base Object Reference=Get Button_EmptySlot() -> (Button_EmptySlot:Common Button Legacy Object Reference)
    InSelected:bool=@n2.bIsSelected
    bGiveClickFeedback:bool=true
   ) -> (then)
   Event HandleLogoutComplete() {is=title K2Node_Event}
   > DeactivateWidget(self=Get WBP_SocialNudge())
   n3 = Event HandleLoginComplete!() -> (OutputDelegate, then, Result:ELoginResult Enum=NotStarted, ErrorReason:"Text (by ref)") {is=title K2Node_Event}
   > Branch(Condition="Equal (Enum)"(A:ELoginResult Enum=@n3.Result, B:ELoginResult Enum=Success))
   > ActivateWidget(self=Get WBP_SocialNudge())
   Event On Focused() {is=title K2Node_Event}
   n4 = IsUsingTouch(PlayerController=GetOwningPlayer())
   > Branch(Condition=AND Boolean(A=@n1, B=NOT Boolean(A=@n4)))
   > SetFocusAndUpdateSocialNudgeState!(
    execute
    self:WBP Social Nudge Object Reference=Get WBP_SocialNudge()
    UseAutoFocus:bool=true
    CurrentFocusedWidget:SocialNudgeFocusState Enum=NewEnumerator0
    IsHovered:bool=true
   ) -> (then)
  ]}
 SetNudgePosition {nodes [
   n1 = SetNudgePosition!() -> (then, Position:int) {is=title K2Node_FunctionEntry}
   > n2 = Set NudgePosition(NudgePosition:int=@n1) -> (Output_Get:int=0)
   n3 = SelectFloat(A=340, B=180, bPickA=IsUsingCompactUI())
   n4 = Select(
    Option 0:bool=IsVisible(self=Get PartySuggestion())
    Option 1:bool=IsSocialPanelVisible!(self:WBP Social Nudge Object Reference=Get WBP_SocialNudge()) -> (Visible:bool)
    Index=Get bUseNewNameplatesAndSocialNudges(self=GetRuntimeOptions())
   ) -> (ReturnValue:bool)
   > n5 = Set!(
    execute
    Padding_Left:float=0
    Padding_Top:float=0
    Padding_Right:float=0
    Padding_Bottom:float=float * float(A:int=@n2, B=Select(Option 0:double=0, Option 1:double=@n3, Index=@n4) -> (ReturnValue:double)) {is=title K2Node_PromotableOperator}
   ) -> (then, Output_Get:Margin Structure)
   > SetPadding!(execute, self:User Widget Object Reference, InPadding:Margin Structure=@n5) -> (then)
  ]}
]
variables [NudgePosition {type=integer}]
widgets {
 info {
  inheritedWidgetCount=3
  namedSlotCount=2
  parentClass=/Script/FortniteUI.FortLobbyAddPlayer
  rootWidgetClass=&CommonActivatableWidgetSwitcher
  widgetCount=10
 }
 widgets [
  Switch_OldNewSocialNudges {widgetClassPath=&CommonActivatableWidgetSwitcher, bIsVariable=true, widget=Switch_OldNewSocialNudges}
  FortHUDElementWrapper {
   widgetClassPath=/Game/Athena/HUD/FortHUDElementWrapper
   bIsVariable=true
   parent=Switch_OldNewSocialNudges
   widget=FortHUDElementWrapper
   properties {elementTag {tagName=HUD}}
   slot {class=WidgetSwitcherSlot}
  }
  VB_NewNudge {widgetClassPath=&VerticalBox, bIsVariable=true, namedSlotHost=FortHUDElementWrapper, widget=VB_NewNudge}
  PartySuggestion {
   widgetClassPath=/Game/Athena/UI/Frontend/Parties/WBP_PartySuggestionButton
   bInherited=true
   bIsVariable=true
   parent=VB_NewNudge
   widget=PartySuggestion
   properties {
    onButtonBaseClicked=()
    onButtonBaseDoubleClicked=()
    onButtonBaseHovered=()
    onButtonBaseUnhovered=()
    onSelectedChangedBase=()
    visibility=Collapsed
   }
   slot {class=VerticalBoxSlot}
  }
  OverlayEmptySlot {
   widgetClassPath=&Overlay
   parent=VB_NewNudge
   widget=OverlayEmptySlot
   slot {class=VerticalBoxSlot, properties {horizontalAlignment=HAlign_Center, verticalAlignment=VAlign_Center, padding=<ltrb 0 40 0 0>}}
  }
  Button_EmptySlot {
   widgetClassPath=/FortUILibrary/Foundation/Buttons/PanelButton
   bInherited=true
   bIsVariable=true
   parent=OverlayEmptySlot
   widget=Button_EmptySlot
   properties {
    bInteractableWhenSelected=true
    bSelectable=true
    bShouldUseFallbackDefaultInputAction=false
    clickMethod=MouseDown
    onButtonBaseClicked=()
    onButtonBaseDoubleClicked=()
    onButtonBaseHovered=()
    onButtonBaseUnhovered=()
    onSelectedChangedBase=()
    style=/FortUILibrary/Foundation/Buttons/ButtonStyles/ButtonStyle-TextOnlyBase_Empty.ButtonStyle-TextOnlyBase_Empty_C
    hoveredSlateSoundOverride {resourceObject=/Game/Sounds/Fort_UI_Sounds/FrontEnd/DefaultSounds/PressedButtons/MSS_UI_Button_Hover_Default}
   }
   slot {class=OverlaySlot}
  }
  WBP_PlayerPad_AddPlayer {
   widgetClassPath=/Game/UI/Frontend/Lobby/PlayerPad/WBP_PlayerPad_AddPlayer
   bIsVariable=true
   namedSlotHost=Button_EmptySlot
   widget=WBP_PlayerPad_AddPlayer
  }
  OverlayInputAction {
   widgetClassPath=&Overlay
   bIsVariable=true
   parent=VB_NewNudge
   widget=OverlayInputAction
   slot {class=VerticalBoxSlot, properties {horizontalAlignment=HAlign_Center, verticalAlignment=VAlign_Top}}
  }
  InputActionWidget {
   widgetClassPath=&CommonActionWidget
   bInherited=true
   parent=OverlayInputAction
   widget=InputActionWidget
   properties {visibility=Hidden, inputActions [{dataTable=&InputActionDataTable, rowName=Select}]}
   slot {class=OverlaySlot, properties {padding=<ltrb 0 24 0 0>}}
  }
  WBP_SocialNudge {
   widgetClassPath=&/WBP_SocialNudge
   bIsVariable=true
   parent=Switch_OldNewSocialNudges
   widget=WBP_SocialNudge
   slot {class=WidgetSwitcherSlot}
  }
 ]
}
